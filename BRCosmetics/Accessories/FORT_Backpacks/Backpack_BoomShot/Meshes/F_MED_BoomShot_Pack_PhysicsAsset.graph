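# Physics asset for the BoomShot backpack cosmetic mesh, F_MED_BoomShot_Pack.
# A single body on the pack bone uses one angled capsule to approximate the backpack's volume.
# It has no constraints, so the pack carries no jointed or simulated secondary parts within this asset.
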PhysicsAsset /BRCosmetics/Accessories/FORT_Backpacks/Backpack_BoomShot/Meshes/F_MED_BoomShot_Pack_PhysicsAsset
bodies {pack [{shapeType=Capsule, length=23.57, radius=11.5696, center=<-3.8863 -9.3704 0.2031>, rotation=<rot 8.4374 -90 -39.7097>}]}
constraints []
